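# Translucent UI material that procedurally draws a drop-down arrow for the age gate login screen, without textures.
# Three rotated, offset UV gradients form the top, left and right edges, each sharpened by a softness multiplier and saturated.
# Dot products of these single-channel values act as multiplies, intersecting them into one triangular mask.
# Angles, offsets and softness are scalar parameters, alongside a Color vector parameter.
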Material /FortniteCoreUI/UI/Login/AgeGate/M_UI_AgeGateDropDownARrow
graph {
 properties {
  bUsedWithStaticMesh=false
  blendMode=BLEND_Translucent
  materialDomain=MD_UI
  pixelDepthOffsetMode=PDOM_Legacy
  usedShadingModels=Default Lit
 }
 nodes [
  n1 = Rotator(Coordinate=TextureCoordinate(), Time=ScalarParameter() {parameterName=TopAngle}) {speed=6.2832}
  n2 = ScalarParameter() {defaultValue=-0.125, parameterName=TopAngleOffset}
  n3 = Multiply(A=ComponentMask(None=Add(A=@n1, B=@n2)) {g=true}, B=ScalarParameter() {defaultValue=48, parameterName=TopSoftness})
  n4 = TextureCoordinate()
  n5 = ScalarParameter() {defaultValue=0.125, parameterName=LeftAngle}
  n6 = ComponentMask(None=Add(A=Rotator(Coordinate=@n4, Time=@n5) {speed=6.2832}, B=ScalarParameter() {defaultValue=-0.125, parameterName=LeftAngleOffset})) {r=true}
  n7 = ScalarParameter() {defaultValue=48, parameterName=Softness}
  n8 = Rotator(Coordinate=TextureCoordinate(), Time=ScalarParameter() {defaultValue=-0.125, parameterName=RightAngle}) {speed=6.2832}
  n9 = ScalarParameter() {defaultValue=0.125, parameterName=RightAngleOffset}
  n10 = ComponentMask(None=OneMinus(None=Add(A=@n8, B=@n9))) {r=true}
  n11 = DotProduct(A=Saturate(None=Multiply(A=@n6, B=@n7)), B=Saturate(None=Multiply(A=@n10, B=@n7)))
  Saturate(None=DotProduct(A=Saturate(None=@n3), B=Saturate(None=@n11)))
  VectorParameter() {parameterName=Color, defaultValue=<rgba 0 1 1 1>}
 ]
}
